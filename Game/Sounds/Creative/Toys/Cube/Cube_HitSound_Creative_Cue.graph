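# Sound cue for hit impacts on the Creative cube toy, reusing small cube impact sounds from the Athena Cube event.
# A Random node picks one of five equally weighted impact waves and feeds it into a Modulator at the root for variation.
# It plays at reduced volume with the cube impact attenuation, a mission-important sound class, and a concurrency override allowing only one instance at a time.
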SoundCue /Game/Sounds/Creative/Toys/Cube/Cube_HitSound_Creative_Cue
root=Modulator_0
nodes [
 {id=Modulator_0, children [Random_0]}
 {id=Random_0, children [WavePlayer_0,WavePlayer_1,WavePlayer_2,WavePlayer_3,WavePlayer_4], properties {weights [1,1,1,1,1]}}
 {id=WavePlayer_0, properties {soundWaveAssetPtr=/Game/Athena/Sounds/Events/Cube/Cube_ImpactSmall_01}}
 {id=WavePlayer_1, properties {soundWaveAssetPtr=/Game/Athena/Sounds/Events/Cube/Cube_ImpactSmall_02}}
 {id=WavePlayer_2, properties {soundWaveAssetPtr=/Game/Athena/Sounds/Events/Cube/Cube_ImpactSmall_03}}
 {id=WavePlayer_3, properties {soundWaveAssetPtr=/Game/Athena/Sounds/Events/Cube/Cube_ImpactSmall_04}}
 {id=WavePlayer_4, properties {soundWaveAssetPtr=/Game/Athena/Sounds/Events/Cube/Cube_ImpactSmall_05}}
]
properties {
 attenuationSettings=/Game/Athena/Sounds/Events/Cube/Cube_Impacts_Att
 bOverrideConcurrency=true
 maxDistance=10000
 soundClassObject=&Fort_Mission_Ingame_Important_PSM
 volumeMultiplier=0.5
 concurrencyOverrides {maxCount=1}
}
