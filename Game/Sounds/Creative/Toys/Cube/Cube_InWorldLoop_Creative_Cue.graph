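# Looping in-world ambience cue for the Creative Cube, limited to one instance and mixed from three layers.
# A quality switch picks full or optimized looping hum pairs, blending distant and close sounds through separate attenuations.
# A looped random delay fires one of two Cube voice clips under area-effect attenuation.
# A pitched-down pair of wind loops forms a non-spatialized bed, shaped by a box attenuation with a custom falloff curve.
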SoundCue /Game/Sounds/Creative/Toys/Cube/Cube_InWorldLoop_Creative_Cue
root=Mixer_2
nodes [
 {id=Attenuation_0, children [WavePlayer_2], properties {attenuationSettings=&/Cube_Creative_Ambient_Att}}
 {id=Attenuation_1, children [WavePlayer_0], properties {attenuationSettings=&/Cube_Creative_Ambient_Close_Att}}
 {id=Attenuation_2, children [WavePlayer_4], properties {attenuationSettings=&Cube_Creative_Ambient_Att}}
 {id=Attenuation_3, children [WavePlayer_3], properties {attenuationSettings=&Cube_Creative_Ambient_Close_Att}}
 {id=Attenuation_5, children [Looping_0], properties {attenuationSettings=&/Cube_Creative_AreaEffect_Att}}
 {
  id=Attenuation_6
  children [Modulator_0]
  properties {
   bOverrideAttenuation=true
   attenuationOverrides {
    attenuationShape=Box
    bAttenuateWithLPF=true
    bSpatialize=false
    falloffDistance=3000
    lPFFrequencyAtMin=10000
    nonFocusVolumeAttenuation=0
    nonSpatializedRadiusStart=2000
    occlusionInterpolationTime=0.5
    occlusionVolumeAttenuation=0.4
    stereoSpread=5000
    attenuationShapeExtents=<7000 7000 2000>
    customAttenuationCurve {editorCurveData {keys [
       <at 0 0>
       {arriveTangent=1.6667, interpMode=RCIM_Cubic, leaveTangent=1.6667, tangentMode=RCTM_Break, time=0.0501, value=0.52}
       {arriveTangent=-1.25, interpMode=RCIM_Cubic, leaveTangent=-1.25, tangentMode=RCTM_Break, time=0.9847, value=0.52}
       <at 1 0>
      ]}}
   }
  }
 }
 {id=Delay_0, children [Random_0], properties {delayMax=15, delayMin=5}}
 {id=Looping_0, children [Delay_0]}
 {id=Mixer_0, children [Attenuation_0, Attenuation_1], properties {inputVolume [0.25,1.1]}}
 {id=Mixer_1, children [Attenuation_2, Attenuation_3], properties {inputVolume [0.25,1.1]}}
 {id=Mixer_2, children [QualityLevel_0, Attenuation_5, Attenuation_6], properties {inputVolume [1,0.5,0.2]}}
 {id=Mixer_4, children [WavePlayer_1, WavePlayer_14], properties {inputVolume [0.3,1]}}
 {id=Modulator_0, children [Mixer_4], properties {pitchMax=0.7, pitchMin=0.7}}
 {id=QualityLevel_0, children [Mixer_0,Mixer_0,Mixer_1,Mixer_1]}
 {id=Random_0, children [WavePlayer_5,WavePlayer_6], properties {weights [1,1]}}
 {id=WavePlayer_0, properties {bLooping=true, soundWaveAssetPtr=/Game/Athena/Sounds/Events/Cube/Cube_AmbientClose_01}}
 {id=WavePlayer_1, properties {bLooping=true, soundWaveAssetPtr=/Game/Sounds/Fort_Ambient_Looping/Wind_Exterior/border_wind_04Loop}}
 {id=WavePlayer_14, properties {bLooping=true, soundWaveAssetPtr=/Game/Athena/Sounds/Ambient/Skytube/Rev1/WindConduit_Ambient_01}}
 {id=WavePlayer_2, properties {bLooping=true, soundWaveAssetPtr=/Game/Athena/Sounds/Events/Cube/Cube_Ambient_01}}
 {id=WavePlayer_3, properties {bLooping=true, soundWaveAssetPtr=/Game/Athena/Sounds/Events/Cube/Cube_AmbientClose_01_opt}}
 {id=WavePlayer_4, properties {bLooping=true, soundWaveAssetPtr=/Game/Athena/Sounds/Events/Cube/Cube_Ambient_01_opt}}
 {id=WavePlayer_5, properties {soundWaveAssetPtr=/Game/Athena/Sounds/Events/Cube/Cube_PrintVoice_02}}
 {id=WavePlayer_6, properties {soundWaveAssetPtr=/Game/Athena/Sounds/Events/Cube/Cube_PrintVoice_01}}
]
properties {
 bOverrideConcurrency=true
 maxDistance=10000
 soundClassObject=/Game/Sounds/FortSoundClassesAndModes/Fort_Events_Scripted
 volumeMultiplier=0.6288
 concurrencyOverrides {maxCount=1}
}
